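# Light function material that, by its name, shapes a caustics lighting pattern for IcicleLarva.
# A high-contrast swirly noise, tiled by U and V parameters, blends between normal and inverted caustic intensities.
# That result multiplies a scaled background pattern mixing electric noise with low-res blurred noise at untiled UVs.
# The graph has no panning or time input, so the pattern is static.
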Material /Game/Blueprints/Lighting/LightFunctions/LF_Caustics_IcicleLarva
graph {
 properties {materialDomain=MD_LightFunction, pixelDepthOffsetMode=PDOM_Legacy, shadingModel=MSM_Subsurface, usedShadingModels=Subsurface}
 nodes [
  n1 = Multiply(
   A=TextureCoordinate()
   B=AppendVector(A=ScalarParameter() {defaultValue=1, parameterName=U}, B=ScalarParameter() {defaultValue=1, parameterName=V})
  )
  n2 = Multiply(
   A=TextureSample() {samplerType=SAMPLERTYPE_LinearColor, texture=/Game/Characters/Player/Male/Medium/Bodies/M_MED_SRDrift/Textures/T_ElectricNoise}.RGB
   B=ScalarParameter() {defaultValue=1, parameterName=PatternBgCausticIntensity}
  )
  n3 = TextureSample() {samplerType=SAMPLERTYPE_Grayscale, texture=&LowResBlurredNoise}
  Multiply(
   A=LinearInterpolate(
    A=ScalarParameter() {defaultValue=0.75, parameterName=CausticIntensity}
    B=ScalarParameter() {defaultValue=2, parameterName=CausticIntensityInvert}
    Alpha=TextureSample(UVs=@n1) {texture=/Game/Weapons/WeaponSkins/Textures/T_Swirly_Noise_highContrast}.R
   ) {constA=0.5, constB=1.5}
   B=Multiply(A=LinearInterpolate(A=@n2, B=@n3.RGB), B=ScalarParameter() {defaultValue=2, parameterName=PatternBgIntensity})
  )
 ]
}
