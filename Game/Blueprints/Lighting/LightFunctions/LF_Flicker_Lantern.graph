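# Light function material that produces an animated flicker, named for lantern lights.
# Time scaled by four speed parameters runs through Frac and Abs to form triangle waves, each weighted by an amplitude parameter.
# Wave channels are summed, offset and floored by a minimum brightness parameter.
# A branch builds a time offset from object pivot or actor position, though as written nothing consumes its result.
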Material /Game/Blueprints/Lighting/LightFunctions/LF_Flicker_Lantern
graph {
 properties {
  bAutomaticallySetUsageInEditor=false
  bForceCompatibleWithLightFunctionAtlas=true
  materialDomain=MD_LightFunction
  pixelDepthOffsetMode=PDOM_Legacy
  usedShadingModels=Default Lit
 }
 nodes [
  n1 = ScalarParameter() {group=A_Emissive_Flicker_V2, parameterName=Use Actor Position?, sortPriority=2}
  n2 = LinearInterpolate(
   A=MaterialFunctionCall() {materialFunction=&ObjectPivotPoint}.Object Pivot Location
   B=ActorPositionWS()
   Alpha=Saturate(None=Round(None=@n1))
  )
  n3 = MaterialFunctionCall(Float3=Frac(None=Multiply(A=@n2) {constB=0.15})) {materialFunction=&BreakOutFloat3Components}
  n4 = Add(A=@n3.R, B=@n3.R)
  n5 = Time()
  Add(A=Add(A=Add(A=@n4, B=@n3.B)), B=@n5)
  n6 = Multiply(
   A=MaterialFunctionCall(
    R=ScalarParameter() {group=A_Emissive_Flicker_V2, parameterName=Flicker Speed 1, sortPriority=1}
    G=ScalarParameter() {defaultValue=0.75, group=A_Emissive_Flicker_V2, parameterName=Flicker Speed 2, sortPriority=1}
    B=ScalarParameter() {defaultValue=0.5, group=A_Emissive_Flicker_V2, parameterName=Flicker Speed 3, sortPriority=1}
    A=ScalarParameter() {defaultValue=0.5, group=A_Emissive_Flicker_V2, parameterName=Flicker Speed 4, sortPriority=1}
   ) {materialFunction=&AppendMany}.RGBA
   B=Multiply(A=@n5)
  )
  n7 = Multiply(
   A=Add(A=Frac(None=@n6)) {constB=-0.5}
   B=MaterialFunctionCall(
    R=ScalarParameter() {defaultValue=0.3, group=A_Emissive_Flicker_V2, parameterName=Flicker Amplitude 1, sortPriority=1}
    G=ScalarParameter() {defaultValue=0.15, group=A_Emissive_Flicker_V2, parameterName=Flicker Amplitude 2, sortPriority=1}
    B=ScalarParameter() {defaultValue=0.25, group=A_Emissive_Flicker_V2, parameterName=Flicker Amplitude 3, sortPriority=1}
    A=ScalarParameter() {defaultValue=0.3, group=A_Emissive_Flicker_V2, parameterName=Flicker Amplitude 4, sortPriority=1}
   ) {materialFunction=&AppendMany}.RGBA
  ) {constB=4}
  n8 = MaterialFunctionCall(Float4=Multiply(A=Abs(None=@n7)) {constB=2}) {materialFunction=&BreakOutFloat4Components}
  n9 = Add(A=@n8.R, B=@n8.R)
  n10 = Add(A=@n8.B, B=@n8.B)
  Max(
   A=Add(
    A=Add(A=@n9, B=@n10)
    B=ScalarParameter() {defaultValue=-0.4, group=A_Emissive_Flicker_V2, parameterName=Amplitude Range Offset, sortPriority=2}
   )
   B=ScalarParameter() {group=A_Emissive_Flicker_V2, parameterName="Minimum Brightness Clamp ( 0-1 Range)", sliderMax=1, sortPriority=2}
  )
 ]
}
