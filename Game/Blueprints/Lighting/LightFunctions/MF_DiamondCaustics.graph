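# Light function material, compatible with the light function atlas, that produces animated diamond caustics from a crystal caustics mask.
# Tiled samples at two scales mix channels: a static G plus a sine-panned R layer, multiplied by a rescaled G.
# The result is raised to an exponent, clamped and remapped between CausticsMin and CausticsMax.
# UV distortion, additive Int 1 and Pan3 branches feed nothing downstream.
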Material /Game/Blueprints/Lighting/LightFunctions/MF_DiamondCaustics
graph {
 properties {
  bAutomaticallySetUsageInEditor=false
  bForceCompatibleWithLightFunctionAtlas=true
  materialDomain=MD_LightFunction
  usedShadingModels=Default Lit
 }
 nodes [
  n1 = TextureCoordinate()
  n2 = Multiply(A=@n1, B=ScalarParameter() {defaultValue=5, parameterName=Tile})
  n3 = TextureObject() {samplerType=SAMPLERTYPE_LinearColor, texture=&/CrystalCaustics_M}
  n4 = Multiply(A=@n2) {constA=0.97, constB=1.1}
  n5 = Panner(Coordinate=@n4, Time=Sine(None=Time()) {period=10}, Speed=VectorParameter() {parameterName=Pan1, defaultValue=<rgba 0.01 0 0 1>}.RGB)
  n6 = Add(
   A=TextureSample(UVs=@n2, Tex=@n3).G
   B=Multiply(A=TextureSample(UVs=@n5, Tex=@n3).R, B=ScalarParameter() {defaultValue=0.8, parameterName=Int 2})
  )
  n7 = TextureSample(UVs=@n4, Tex=@n3)
  Add(A=@n6, B=Multiply(A=@n7.G, B=ScalarParameter() {parameterName=Int 1}))
  n8 = TextureCoordinate()
  n9 = TextureSample(
   UVs=Panner(Coordinate=Multiply(A=@n8) {constB=2}, Speed=VectorParameter() {parameterName=Panx, defaultValue=<rgba -0.001 -0.0003 0 1>}.RGB)
   Tex=@n3
  ) {texture=&CrystalCaustics_M}
  Add(A=@n1, B=Multiply(A=AppendVector(A=@n9.B, B=@n9.B)) {constB=0.2})
  n10 = Multiply(A=@n6, B=@n7.G)
  n11 = ScalarParameter() {defaultValue=2, parameterName=Exp_2}
  LinearInterpolate(
   A=ScalarParameter() {defaultValue=0.25, parameterName=CausticsMin}
   B=ScalarParameter() {defaultValue=1, parameterName=CausticsMax}
   Alpha=Clamp(None=Power(Base=@n10, Exp=@n11))
  )
  Multiply(A=@n2) {constB=0.9}
  Panner(Coordinate=@n2, Speed=VectorParameter() {parameterName=Pan3, defaultValue=<rgba 0 0 0 1>}.RGB)
 ]
}
