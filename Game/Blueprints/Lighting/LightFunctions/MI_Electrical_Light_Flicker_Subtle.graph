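# Light function material, named for a subtle electrical flicker, that samples a looping brightness curve to flicker lights.
# Time times a speed, wrapped with Frac, indexes a curve atlas row parameter that defaults to a fire flicker curve.
# The speed blends between two flicker speed parameters by the fractional part of RandomLightVal, allowing per-light variation.
# The material is forced compatible with the light function atlas.
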Material /Game/Blueprints/Lighting/LightFunctions/MI_Electrical_Light_Flicker_Subtle
graph {
 properties {
  bAutomaticallySetUsageInEditor=false
  bForceCompatibleWithLightFunctionAtlas=true
  materialDomain=MD_LightFunction
  pixelDepthOffsetMode=PDOM_Legacy
  usedShadingModels=Default Lit
 }
 nodes [
  CurveAtlasRowParameter()
  n1 = Time()
  n2 = LinearInterpolate(
   A=ScalarParameter() {defaultValue=0.05, parameterName=FlickerSpeed}
   B=ScalarParameter() {defaultValue=0.3, parameterName=FlickerSpeed2}
   Alpha=Frac(None=ScalarParameter() {defaultValue=0.5, parameterName=RandomLightVal})
  )
  CurveAtlasRowParameter(CurveTime=Frac(None=Multiply(A=@n1, B=@n2))) {
   atlas=/Game/Environments/Helios/Props/Brimstone/Brimstone_Brazier_A/Blueprints/CurveAtlas_FN_LightFunctions
   curve=&/Curve_FireFlickers
   parameterName=FireFlicker
  }
 ]
}
