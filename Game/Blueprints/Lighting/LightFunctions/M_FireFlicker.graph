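# Light function material that makes a light flicker like fire, flagged compatible with the light function atlas.
# Time drives a FireFlicker curve atlas lookup, at a rate blended between two speed parameters by RandomLightVal to vary lights.
# The curve multiplies vertically panning Perlin noise, confined by a sphere mask to a band around mid-V, then clamps to a Min floor.
# Several leftover nodes feed nothing.
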Material /Game/Blueprints/Lighting/LightFunctions/M_FireFlicker
graph {
 properties {
  bForceCompatibleWithLightFunctionAtlas=true
  materialDomain=MD_LightFunction
  numCustomizedUVs=1
  pixelDepthOffsetMode=PDOM_Legacy
  usedShadingModels=Default Lit
 }
 nodes [
  n1 = TextureCoordinate()
  Abs(None=MaterialFunctionCall(UV Channel=@n1) {materialFunction=&LinearGradient}.VGradient)
  n2 = TextureCoordinate() {vTiling=0.2}
  n3 = Frac(None=ScalarParameter() {defaultValue=0.5, parameterName=RandomLightVal})
  Add(A=@n2, B=@n3)
  n4 = Time()
  n5 = LinearInterpolate(
   A=ScalarParameter() {defaultValue=0.75, parameterName=FlickerSpeed}
   B=ScalarParameter() {defaultValue=0.133, parameterName=FlickerSpeed2}
   Alpha=@n3
  )
  n6 = CurveAtlasRowParameter(CurveTime=Frac(None=Multiply(A=@n4, B=@n5))) {
   atlas=/Game/Environments/Helios/Props/Brimstone/Brimstone_Brazier_A/Blueprints/CurveAtlas_FN_LightFunctions
   curve=&/Curve_FireFlickers
   parameterName=FireFlicker
  }
  n7 = Constant() {r=0.5}
  n8 = ScalarParameter() {defaultValue=0.25, parameterName=Min}
  n9 = TextureSample(UVs=Panner(Coordinate=@n2) {speedY=-0.33}) {texture=/Game/Global/Textures/Default/Noise/3xPerlin_M}
  n10 = LinearInterpolate(A=@n8, Alpha=@n9.R) {constA=0.66}
  n11 = SphereMask(A=ComponentMask(None=@n1) {g=true}, B=Constant() {r=0.5}) {attenuationRadius=0.3, hardnessPercent=1}
  Clamp(None=Multiply(A=@n6.R, B=LinearInterpolate(A=@n7, B=@n10, Alpha=@n11)), Min=@n8)
  MaterialFunctionCall() {materialFunction=&LocalPosition}
  Multiply(A=@n9.R)
  Saturate(None=ComponentMask(None=@n3) {r=true})
 ]
}
